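# Sound cue for saddle activation in the riding content, judging by its name and folder.
# A single wave player feeds the Saddle_Activate_01 wave straight to the output at reduced volume, with no randomization or mixing.
# It routes through the Fort_Gadgets_Items sound class and references the shared Fort_Saddler_Att attenuation asset.
# Local attenuation override values are stored, but the override flag is not enabled.
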SoundCue /RidingContent/Sounds/Saddle/Saddle_Activate_Cue
root=WavePlayer_2
nodes [{id=WavePlayer_2, properties {soundWaveAssetPtr=&/Saddle_Activate_01}}]
properties {
 attenuationSettings=/Game/Sounds/Attenuation/Fort_Saddler_Att
 maxDistance=4500
 soundClassObject=&Fort_Gadgets_Items
 volumeMultiplier=0.6
 attenuationOverrides {distanceAlgorithm=Logarithmic, falloffDistance=500, attenuationShapeExtents {x=0}}
}
