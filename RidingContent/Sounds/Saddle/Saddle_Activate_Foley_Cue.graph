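# Sound cue for the foley played when a saddle activates in the riding content.
# A single wave player feeds the Saddle_Activate_Foley_01 wave straight to the output at reduced volume, with no randomization or mixing.
# It routes to the gadgets and items sound class and references an item-interaction attenuation asset for distance falloff.
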SoundCue /RidingContent/Sounds/Saddle/Saddle_Activate_Foley_Cue
root=WavePlayer_0
nodes [{id=WavePlayer_0, properties {soundWaveAssetPtr=&/Saddle_Activate_Foley_01}}]
properties {
 attenuationSettings=/Game/Sounds/Attenuation/Fort_Item_Interact_Att
 maxDistance=1850
 soundClassObject=&Fort_Gadgets_Items
 volumeMultiplier=0.4
 attenuationOverrides {distanceAlgorithm=Logarithmic, falloffDistance=500, attenuationShapeExtents {x=0}}
}
